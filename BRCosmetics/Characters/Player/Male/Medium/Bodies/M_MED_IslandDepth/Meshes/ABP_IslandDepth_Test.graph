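# Test Anim Blueprint for the IslandDepth body that retargets the pose from a mesh through IKR_IslandDepth.
# The retarget node receives slope data plus attach sockets and bounds for weapons, glider and four root attach volumes.
# Pawn weapon and emote events, spawn-prop notifies and per-update scans of attached parachutes and props refresh those inputs.
# Weapon mesh or hand-prop counts choose the weapon mode, with the building tool special-cased as dual wield.
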AnimBlueprint /BRCosmetics/Characters/Player/Male/Medium/Bodies/M_MED_IslandDepth/Meshes/ABP_IslandDepth_Test
target_skeleton=&/M_MED_IslandDepth_Skeleton
graphs [
 AnimGraph {nodes [
   n1 = Break Fort Anim Input Slope Info!(FortAnimInput_SlopeInfo:"Fort Anim Input Slope Info Structure (by ref)"=Get SlopeInfoInput() -> (SlopeInfoInput:Fort Anim Input Slope Info Structure)) -> (RootSlopeTranslation:Vector=~, LeftFootSlopeTranslation:Vector=~, RightFootSlopeTranslation:Vector=~, SlopeRotation:Rotator=~) {is=title K2Node_BreakStruct}
   n2 = BreakBoxSphereBounds!(InBoxSphereBounds:"Box Sphere Bounds Structure (by ref)"=Get Weapon Bounds() -> (Weapon Bounds:Box Sphere Bounds Structure)) -> (Origin:Vector, BoxExtent:Vector, SphereRadius:float)
   n3 = BreakBoxSphereBounds(InBoxSphereBounds=Get WeaponB Bounds() -> (WeaponB Bounds:Box Sphere Bounds Structure))
   n4 = BreakBoxSphereBounds(InBoxSphereBounds=Get Glider Bounds() -> (Glider Bounds:Box Sphere Bounds Structure))
   n5 = Get Root Attach Volumes() -> (Root Attach Volumes:Array of Box Sphere Bounds Structures)
   n6 = BreakBoxSphereBounds(InBoxSphereBounds="Get (a ref)"(Array:Array of Box Sphere Bounds Structures=@n5, Dimension 1=3) -> (Output:"Box Sphere Bounds Structure (by ref)"))
   n7 = BreakBoxSphereBounds(InBoxSphereBounds="Get (a ref)"(Array=@n5, Dimension 1=0))
   n8 = BreakBoxSphereBounds(InBoxSphereBounds="Get (a ref)"(Array=@n5, Dimension 1=1))
   n9 = BreakBoxSphereBounds(InBoxSphereBounds="Get (a ref)"(Array=@n5, Dimension 1=2))
   n10 = Retarget Pose From Mesh!(
    weapon_attach_socket_name:Name=Get Weapon Attach Socket Name() -> (Weapon Attach Socket Name:Name=None)
    weapon_origin:Vector=@n2
    weapon_box_extent:Vector=@n2.BoxExtent
    weapon_sphere_radius:float=@n2.SphereRadius
    weapon_B_attach_socket_name:Name=Get Weapon B Attach Socket Name() -> (Weapon B Attach Socket Name:Name=None)
    weapon_B_origin:Vector=@n3
    weapon_B_box_extent:Vector=@n3.BoxExtent
    weapon_B_sphere_radius:float=@n3.SphereRadius
    weapon_mode:EWeaponMode Enum=Get Weapon Mode() -> (Weapon Mode:EWeaponMode Enum=None)
    ground_location:Vector=@n1
    ground_rotation:Rotator=@n1.SlopeRotation
    ground_left_foot_position:Vector=@n1.LeftFootSlopeTranslation
    ground_right_foot_position:Vector=@n1.RightFootSlopeTranslation
    glider_attach_socket_name:Name="Get Glider  Attach Bone"() -> ("Glider  Attach Bone":Name=None)
    glider_origin:Vector=@n4
    glider_box_extent:Vector=@n4.BoxExtent
    glider_sphere_radius:float=@n4.SphereRadius
    attach_origin:Vector=@n7
    attach_box_extent:Vector=@n7.BoxExtent
    attach_sphere_radius:float=@n7.SphereRadius
    attach_emote_01_origin:Vector=@n8
    attach_emote_01_box_extent:Vector=@n8.BoxExtent
    attach_emote_01_sphere_radius:float=@n8.SphereRadius
    attach_emote_02_origin:Vector=@n9
    attach_emote_02_box_extent:Vector=@n9.BoxExtent
    attach_emote_02_sphere_radius:float=@n9.SphereRadius
    attach_emote_03_origin:Vector=@n6
    attach_emote_03_box_extent:Vector=@n6.BoxExtent
    attach_emote_03_sphere_radius:float=@n6.SphereRadius
    enable_root_attach_pins:bool=Get Enable Root Attach Pins() -> (Enable Root Attach Pins:bool)
   ) -> (Pose:Pose Link Structure) {
    is=title AnimGraphNode_RetargetPoseFromMesh
    bSuppressWarnings=true
    iKRetargeterAsset=&/IKR_IslandDepth
    lODThresholdForIK=1
   }
   Output Pose(Result=@n10)
  ]}
 EventGraph {nodes [
   Emote Started!() -> (OutputDelegate, then, MontageItemDef:Fort Item Definition Object Reference, PawnEmoting:Fort Pawn Object Reference) {is=title K2Node_CustomEvent}
   > Set Enable Root Attach Pins(Enable Root Attach Pins:bool=false) -> (Output_Get:bool)
   > DelayUntilNextTick!(execute) -> (then)
   > Set Enable Root Attach Pins(Enable Root Attach Pins=true)
   n1 = Weapon Visibility Change!() -> (OutputDelegate, then, bVisible:bool=~) {is=title K2Node_CustomEvent}
   > n2 = Branch(Condition=@n1.bVisible)
   > Set Weapon Mode(Weapon Mode:EWeaponMode Enum=Get Current Weapon Mode() -> (Current Weapon Mode:EWeaponMode Enum=None)) -> (Output_Get:EWeaponMode Enum=None)
   Set Weapon Mode(execute=@n2.else, Weapon Mode=None) -> (Output_Get=None)
   Emote Stopped!() -> (OutputDelegate, then, MontageItemDef:Fort Item Definition Object Reference, PawnEmoting:Fort Pawn Object Reference) {is=title K2Node_CustomEvent}
   > n3 = Sequence()
   n4 = Get Weapon Mode!(execute=@n3.then_1, self, Weapon:Fort Weapon Object Reference=Get CurrentWeapon() -> (CurrentWeapon:Fort Weapon Object Reference)) -> (
    then
    Weapon Mode:EWeaponMode Enum=None
    Weapon Attach Socket Name:Name=None
    Weapon Bounds:Box Sphere Bounds Structure
    Weapon B Attach Socket Name:Name=None
    Weapon B Bounds:Box Sphere Bounds Structure
   ) {is=title K2Node_CallFunction}
   > Set Weapon Attach Socket Name(Weapon Attach Socket Name:Name=@n4.Weapon Attach Socket Name) -> (Output_Get:Name=None)
   > Set Weapon Bounds(Weapon Bounds:Box Sphere Bounds Structure=@n4.Weapon Bounds) -> (Output_Get:Box Sphere Bounds Structure)
   > Set Weapon B Attach Socket Name(Weapon B Attach Socket Name:Name=@n4.Weapon B Attach Socket Name) -> (Output_Get:Name=None)
   > Set WeaponB Bounds(WeaponB Bounds:Box Sphere Bounds Structure=@n4.Weapon B Bounds) -> (Output_Get:Box Sphere Bounds Structure)
   > n5 = Set Weapon Mode(Weapon Mode=@n4) -> (Output_Get=None)
   > Set Current Weapon Mode(Current Weapon Mode:EWeaponMode Enum=@n5) -> (Output_Get:EWeaponMode Enum=None)
   Clear(execute=@n3.then_0, TargetArray:Array of Integers=Get Hand Skeletal Mesh Prop() -> (Hand Skeletal Mesh Prop:Array of Integers))
   > Clear(TargetArray=Get Hand Prop ID() -> (Hand Prop ID:Array of Integers))
   > Clear(TargetArray=Get World Skeletal Mesh Prop() -> (World Skeletal Mesh Prop:Array of Integers))
   > Clear(TargetArray=Get World Prop ID() -> (World Prop ID:Array of Integers))
   n6 = MakeBoxSphereBounds!(Origin:Vector="0, 0, 0", BoxExtent:Vector="0, 0, 0", SphereRadius:float=0) -> (ReturnValue:Box Sphere Bounds Structure)
   > Set Root Attach Volumes(Root Attach Volumes:Array of Box Sphere Bounds Structures=Make Array!(
     "[0]":Box Sphere Bounds Structure=@n6
     "[1]":Box Sphere Bounds Structure=@n6
     "[2]":Box Sphere Bounds Structure=@n6
     "[3]":Box Sphere Bounds Structure=@n6
    ) -> (Array:Array of Box Sphere Bounds Structures)) -> (Output_Get:Array of Box Sphere Bounds Structures)
   n7 = Weapon Equipped!() -> (OutputDelegate, then, NewWeapon:Fort Weapon Object Reference, PrevWeapon:Fort Weapon Object Reference) {is=title K2Node_CustomEvent}
   > n8 = Sequence()
   > n9 = Get Weapon Mode(Weapon=@n7.NewWeapon) -> (Weapon Mode=None, Weapon Attach Socket Name=None, Weapon B Attach Socket Name=None)
   > Set Weapon Attach Socket Name(Weapon Attach Socket Name=@n9.Weapon Attach Socket Name) -> (Output_Get=None)
   > Set Weapon Bounds(Weapon Bounds=@n9.Weapon Bounds)
   > Set Weapon B Attach Socket Name(Weapon B Attach Socket Name=@n9.Weapon B Attach Socket Name) -> (Output_Get=None)
   > Set WeaponB Bounds(WeaponB Bounds=@n9.Weapon B Bounds)
   > n10 = Set Weapon Mode(Weapon Mode=@n9) -> (Output_Get=None)
   > Set Current Weapon Mode(Current Weapon Mode=@n10) -> (Output_Get=None)
   n11 = Create Event()
   Unbind Event from On Visibility Changed(execute=@n8.then_1, self:Fort Weapon Object Reference=@n7.PrevWeapon, Delegate=@n11) {is=title K2Node_RemoveDelegate}
   > Bind Event to On Visibility Changed(self:Fort Weapon Object Reference=@n7.NewWeapon, Delegate=@n11) {is=title K2Node_AddDelegate}
   Weapon Holstered() {is=title K2Node_CustomEvent}
   > Set Enable Root Attach Pins(Enable Root Attach Pins=false)
   Event BlueprintUpdateAnimation()
   > n12 = Sequence!(execute) -> (then_0, then_1, then_2:Exec)
   > n13 = For Each Loop(Array:Array of Actor Object References=GetAttachedActors!(self:Actor Object Reference=TryGetPawnOwner(), bResetArray:bool=true, bRecursivelyIncludeAttachedActors:bool=false) -> (OutActors:Array of Actor Object References)) -> (Array Element:Actor Object Reference)
   > n14 = Cast To FortPlayerParachute!(execute, Object=@n13) -> (then, CastFailed, AsFort Player Parachute:Fort Player Parachute Object Reference) {is=title K2Node_DynamicCast}
   n15 = Get ParachuteMesh!(self:Fort Player Parachute Object Reference=@n14) -> (ParachuteMesh:Skeletal Mesh Component Budgeted Object Reference)
   > "Set Glider  Attach Bone"("Glider  Attach Bone":Name=GetAttachSocketName!(self:Scene Component Object Reference=@n15) -> (ReturnValue:Name=None)) -> (Output_Get:Name=None)
   > Set Glider Bounds(Glider Bounds:Box Sphere Bounds Structure=GetBounds!(self:Skeletal Mesh Object Reference=Get SkeletalMesh!(self:Skinned Mesh Component Object Reference=@n15) -> (SkeletalMesh:Skeletal Mesh Object Reference)) -> (ReturnValue:Box Sphere Bounds Structure)) -> (Output_Get:Box Sphere Bounds Structure)
   Branch(execute=@n12.then_1, Condition=Is Not Empty!(TargetArray:Array of Integers=Get Hand Prop ID()) -> (ReturnValue:bool))
   n16 = Get Hand Prop ID()
   n17 = FindSpawnProp!(
    self:Custom Character Part Anim Instance Object Reference
    PropId:int="Get (a ref)"(Array:Array of Integers=@n16, Dimension 1=0) -> (Output:"Integer (by ref)")
   ) -> (ReturnValue:Scene Component Object Reference)
   > Is Valid(InputObject=@n17)
   > Set Weapon Bounds(Weapon Bounds=GetLocalBounds!(Component:Scene Component Object Reference=@n17) -> (ReturnValue:Box Sphere Bounds Structure))
   > Set Weapon Attach Socket Name(Weapon Attach Socket Name=GetAttachSocketName(self=@n17) -> (ReturnValue=None)) -> (Output_Get=None)
   n18 = Get Hand Prop ID()
   > Branch(Condition="integer > integer"(A=Length(TargetArray:Array of Integers=@n18), B=1))
   n19 = FindSpawnProp(PropId="Get (a ref)"(Array=@n16, Dimension 1=1))
   > Is Valid(InputObject=@n19)
   > Set WeaponB Bounds(WeaponB Bounds=GetLocalBounds(Component=@n17))
   > Set Weapon B Attach Socket Name(Weapon B Attach Socket Name=GetAttachSocketName(self=@n19) -> (ReturnValue=None)) -> (Output_Get=None)
   n20 = For Each Loop(Exec=@n12.then_2, Array:Array of Integers=Get World Skeletal Mesh Prop()) -> (Array Element:int)
   n21 = FindSpawnProp(PropId=@n20)
   > n22 = Switch on Name!(execute:Exec, Selection:Name=GetAttachSocketName(self=@n21) -> (ReturnValue=None)) -> (Default:Exec, attach:Exec, attach_emote_01:Exec, attach_emote_02:Exec, attach_emote_03:Exec) {is=title K2Node_SwitchName, pinNames [attach, attach_emote_01, attach_emote_02, attach_emote_03]}
   n23 = GetLocalBounds(Component=@n21)
   Set Array Elem!(
    execute=@n22.attach
    TargetArray:Array of Box Sphere Bounds Structures=Get Root Attach Volumes()
    Index:int=0
    Item:"Box Sphere Bounds Structure (by ref)"=@n23
    bSizeToFit:bool=true
   ) -> (then)
   Set Array Elem(execute=@n22.attach_emote_01, TargetArray=Get Root Attach Volumes(), Index=1, Item=@n23, bSizeToFit=true)
   Set Array Elem(execute=@n22.attach_emote_02, TargetArray=Get Root Attach Volumes(), Index=2, Item=@n23, bSizeToFit=true)
   Set Array Elem(execute=@n22.attach_emote_03, TargetArray=Get Root Attach Volumes(), Index=3, Item=@n23, bSizeToFit=true)
   Event BlueprintInitializeAnimation()
   > n24 = Sequence()
   > n25 = Cast To FortPawn!(execute, Object=GetOwningActor()) -> (then, CastFailed, AsFort Pawn:Fort Pawn Object Reference) {is=title K2Node_DynamicCast}
   > Bind Event to On Weapon Equipped Delegate(self:Fort Pawn Object Reference=@n25, Delegate=Create Event()) {is=title K2Node_AddDelegate}
   > Bind Event to On Holstered Event(self:Fort Pawn Object Reference=@n25, Delegate=Create Event()) {is=title K2Node_AddDelegate}
   n26 = Is Valid(exec=[@n27.then, @n24.then_1], InputObject=Get CurrentWeapon())
   n27 = Delay(execute=@n26.Is Not Valid, Duration=0.2)
   n28 = Get Weapon Mode(execute=@n26.Is Valid, Weapon=Get CurrentWeapon()) -> (Weapon Mode=None, Weapon Attach Socket Name=None, Weapon B Attach Socket Name=None)
   > Set Weapon Attach Socket Name(Weapon Attach Socket Name=@n28.Weapon Attach Socket Name) -> (Output_Get=None)
   > Set Weapon Bounds(Weapon Bounds=@n28.Weapon Bounds)
   > Set Weapon B Attach Socket Name(Weapon B Attach Socket Name=@n28.Weapon B Attach Socket Name) -> (Output_Get=None)
   > Set WeaponB Bounds(WeaponB Bounds=@n28.Weapon B Bounds)
   > n29 = Set Weapon Mode(Weapon Mode=@n28) -> (Output_Get=None)
   > Set Current Weapon Mode(Current Weapon Mode=@n29) -> (Output_Get=None)
   n30 = Cast To FortPawn(execute=@n24.then_2, Object=GetOwningActor())
   > Bind Event to On Pawn Started Emote(self:Fort Pawn Object Reference=@n30, Delegate=Create Event()) {is=title K2Node_AddDelegate}
   > Bind Event to On Pawn Stopped Emote(self:Fort Pawn Object Reference=@n30, Delegate=Create Event()) {is=title K2Node_AddDelegate}
   n31 = Event OnSpawnPropBegin!() -> (OutputDelegate, then, PropId:int=0) {is=title K2Node_Event}
   > n32 = Sequence()
   n33 = FindSpawnProp(PropId=@n31.PropId)
   Cast To SkeletalMeshComponentBudgeted!(execute=@n32.then_1, Object=@n33) -> (then, CastFailed, AsSkeletal Mesh Component Budgeted:Skeletal Mesh Component Budgeted Object Reference) {is=title K2Node_DynamicCast}
   n34 = GetAttachSocketName(self=@n33) -> (ReturnValue=None)
   n35 = Get World Attach Bones() -> (World Attach Bones:Array of Names)
   > Add(
    TargetArray:Array of Integers=Select(
     Option 0:Array of Integers=Get Hand Skeletal Mesh Prop()
     Option 1:Array of Integers=Get World Skeletal Mesh Prop()
     Index=Contains Item(TargetArray=@n35, ItemToFind=@n34)
    ) -> (ReturnValue:Array of Integers)
    NewItem:"Integer (by ref)"=@n31.PropId
   ) {is=title K2Node_CallArrayFunction}
   n36 = GetAttachSocketName(self=@n33) -> (ReturnValue=None)
   n37 = Branch(execute=@n32.then_2, Condition=Contains Item(TargetArray=Get Hand Attach Bones() -> (Hand Attach Bones:Array of Names), ItemToFind=@n36))
   > Add(TargetArray=Get Hand Prop ID(), NewItem=@n31.PropId)
   > n38 = Switch on Int!(execute:Exec, Selection:int=Length(TargetArray=Get Hand Prop ID())) -> (Default:Exec, 0:Exec, 1:Exec, 2:Exec)
   Set Weapon Mode(execute=@n38.0, Weapon Mode=None) -> (Output_Get=None)
   Set Weapon Mode(execute=@n38.1, Weapon Mode=TwoHanded) -> (Output_Get=None)
   Set Weapon Mode(execute=@n38.2, Weapon Mode=DualWield) -> (Output_Get=None)
   Set Weapon Mode(execute=@n38.Default, Weapon Mode=DualWield) -> (Output_Get=None)
   Branch(execute=@n37.else, Condition=Contains Item(TargetArray=Get World Attach Bones(), ItemToFind=@n36))
   > Add(TargetArray=Get World Prop ID(), NewItem=@n31.PropId)
   n39 = Event OnSpawnPropEnd!() -> (OutputDelegate, then, PropId:int=0) {is=title K2Node_Event}
   > n40 = Sequence!(execute) -> (then_0, then_1)
   > n41 = Sequence!(execute) -> (then_0, then_1, then_2:Exec)
   > Remove Item!(execute, TargetArray:Array of Integers=Get Hand Prop ID(), Item:"Integer (by ref)"=@n39.PropId) -> (then, ReturnValue:bool)
   n42 = Switch on Int(execute=@n41.then_2, Selection=Length(TargetArray=Get Hand Prop ID()))
   Set Weapon Mode(execute=@n42.0, Weapon Mode=None) -> (Output_Get=None)
   Set Weapon Mode(execute=@n42.1, Weapon Mode=TwoHanded) -> (Output_Get=None)
   Set Weapon Mode(execute=@n42.2, Weapon Mode=DualWield) -> (Output_Get=None)
   Set Weapon Mode(execute=@n42.Default, Weapon Mode=DualWield) -> (Output_Get=None)
   Remove Item(execute=@n41.then_1, TargetArray=Get Hand Skeletal Mesh Prop(), Item=@n39.PropId)
   Remove Item(execute=@n40.then_1, TargetArray=Get World Prop ID(), Item=@n39.PropId)
   > Remove Item(TargetArray=Get World Skeletal Mesh Prop(), Item=@n39.PropId)
   GetComponentBounds!(Component:Scene Component Object Reference) -> (Origin:Vector, BoxExtent:Vector, SphereRadius:float)
   GetImportedBounds!(self:Skeletal Mesh Object Reference=GetSkeletalMeshAsset!(self:Skeletal Mesh Component Object Reference=Invalid Message Node!(execute:Exec, self:Object Reference) -> (then:Exec, ReturnValue:Skeletal Mesh Component Object Reference) {is=title K2Node_Message}) -> (ReturnValue:Skeletal Mesh Object Reference)) -> (ReturnValue:Box Sphere Bounds Structure)
   "Equal (Name)"(A=@n36, B=righthand)
  ]}
 Get Weapon Mode {nodes [
   n1 = Get Weapon Mode!() -> (then, Weapon:Fort Weapon Object Reference) {is=title K2Node_FunctionEntry}
   n2 = GetWeaponMeshes!(self:Fort Weapon Object Reference=@n1) -> (WeaponMeshes:Array of Skeletal Mesh Component Budgeted Object References)
   > n3 = Branch(Condition="integer > integer"(A=Length(TargetArray:Array of Skeletal Mesh Component Budgeted Object References=@n2), B=1))
   n4 = "Get (a copy)"(Array:Array of Skeletal Mesh Component Budgeted Object References=@n2, Dimension 1=1) -> (Output:Skeletal Mesh Component Budgeted Object Reference)
   n5 = "Get (a copy)"(Array=@n2)
   n6 = GetAttachSocketName(self=@n5) -> (ReturnValue=None)
   n7 = GetLocalBounds(Component=@n5)
   > Return Node!(
    execute
    Weapon Mode:EWeaponMode Enum=DualWield
    Weapon Attach Socket Name:Name=@n6
    Weapon Bounds:Box Sphere Bounds Structure=@n7
    Weapon B Attach Socket Name:Name=GetAttachSocketName(self=@n4) -> (ReturnValue=None)
    Weapon B Bounds:Box Sphere Bounds Structure=GetLocalBounds(Component=@n4)
   )
   n8 = Get Weapon() -> (Weapon:Fort Weapon Object Reference)
   n9 = Branch(execute=@n3.else, Condition=Contains(SearchIn=GetDisplayName(Object=@n8), Substring=DefaultBuildingTool))
   > Return Node(Weapon Mode=DualWield, Weapon Attach Socket Name=BluePrintPaper, Weapon B Attach Socket Name=BluePrintPen)
   Return Node(execute=@n9.else, Weapon Mode=TwoHanded, Weapon Attach Socket Name=@n6, Weapon Bounds=@n7, Weapon B Attach Socket Name=None)
  ]}
]
variables [
 Box Extent {type=Vector}
 Current Weapon Mode {type=EWeaponMode, default=None}
 Enable Root Attach Pins {type=boolean}
 Find Emote Attachment Timer {type=TimerHandle}
 "Glider  Attach Bone" {type=string, default=None}
 Glider Bounds {type=BoxSphereBounds, default {sphereRadius=0, boxExtent=<0 0 0>, origin=<0 0 0>}}
 Hand Attach Bones {type="array<string>", default [weapon_l, hand_attach_l, LeftHand, weapon_r, hand_attach_r, RightHand, lefthand, righthand]}
 Hand Prop ID {type="array<integer>"}
 Hand Skeletal Mesh Prop {type="array<integer>"}
 Origin {type=Vector}
 Root Attach Volumes {
  type="array<BoxSphereBounds>"
  default [
   {sphereRadius=0, boxExtent=<0 0 0>, origin=<0 0 0>}
   {sphereRadius=0, boxExtent=<0 0 0>, origin=<0 0 0>}
   {sphereRadius=0, boxExtent=<0 0 0>, origin=<0 0 0>}
   {sphereRadius=0, boxExtent=<0 0 0>, origin=<0 0 0>}
  ]
 }
 Sphere Radius {type=number}
 temp_prop_attach_name {type=string, default=None}
 Weapon Attach Socket Name {type=string, default=None}
 Weapon B Attach Socket Name {type=string, default=None}
 Weapon Bounds {type=BoxSphereBounds, default {sphereRadius=0, boxExtent=<0 0 0>, origin=<0 0 0>}}
 Weapon Mode {type=EWeaponMode, default=None}
 WeaponB Bounds {type=BoxSphereBounds, default {sphereRadius=0, boxExtent=<0 0 0>, origin=<0 0 0>}}
 World Attach Bones {type="array<string>", default [attach, attach_emote_01, attach_emote_02, attach_emote_03, root]}
 World Prop ID {type="array<integer>"}
 World Skeletal Mesh Prop {type="array<integer>"}
]
