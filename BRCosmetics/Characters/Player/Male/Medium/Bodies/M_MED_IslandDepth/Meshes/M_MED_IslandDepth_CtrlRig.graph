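# Control Rig for the M_MED_IslandDepth player body that adds corrective deformation to helper bones around the shoulders, torso and hips.
# The forward solve runs the shared DeformRig function, then spherical pose readers on each upper arm drive offsets on custom upperarm, armpit volume and upperarm twist bones.
# Averaged parent constraints blend the scapula, belly and above-leg mass bones between neighbours.
# Remapped thigh-to-pelvis Euler Z angles push each above-leg mass bone, with the right side using 180 minus the absolute angle.
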ControlRigBlueprint /BRCosmetics/Characters/Player/Male/Medium/Bodies/M_MED_IslandDepth/Meshes/M_MED_IslandDepth_CtrlRig
graphs [
 RigVMFunctionLibrary {links [], nodes []}
 RigVMModel {nodes [
   RigUnit_BeginExecution = &RigUnit_BeginExecution()
   FN_DeformRig_FNC = RigVMFunctionReferenceNode(
    DeformRig_ Curve="(deformRig_l_upperarm_up_alpha_72_#1=False,deformRig_r_upperarm_up_alpha_73_#2=False,deformRig_l_upperarm_fr_alpha_74_#3=False,deformRig_r_upperarm_fr_alpha_75_CB51B8 ...(truncated)"
    DeformRig_Bool="(upperarm_1_#4=True,elbow_3_#5=True,wrist_7_#6=True,groin_8_#7=True,glute_17_#8=True,knee_9_40C408324CA261 ...(truncated)"
    DeformRig_Debug=(upperarm_debug_18_#4=False,wrist_debug_20_#6=False,lat_debug_26_#9=False,groinGlute_debug_27_#7=False)
    DeformRig_Value="(upperarm_up_18_#10=(Rotation=(X=0,Y=0,Z=0,W=1),Translation=(X=0,Y=0,Z=8),Scale3D=(X=0.7,Y=1,Z=1)),upperarm_fwd_fr_19_#11=(Rotation=(X=0,Y=0,Z=0,W=1),Translation=(X=0,Y=10,Z=0),Scale3D=(X=0.5,Y=1,Z= ...(truncated)"
    ExecuteContext=@RigUnit_BeginExecution.ExecutePin
   )
   > n1 = &RigUnit_SphericalPoseReader(
    ActiveRegionScaleFactors=(PositiveWidth=1.5,NegativeWidth=7,PositiveHeight=3,NegativeHeight=3)
    ActiveRegionSize=0.07
    Debug=(bDrawDebug=false,bDraw2D=false,bDrawLocalAxes=false,DebugScale=25,DebugSegments=20,DebugThickness=0.25)
    DriverAxis=(X=1,Y=0,Z=0)
    DriverItem=(Type=Bone,Name="upperarm_l")
    FalloffRegionScaleFactors=(PositiveWidth=1,NegativeWidth=1,PositiveHeight=1.5,NegativeHeight=1.5)
    FalloffSize=0.4
    FlipHeightScaling=False
    FlipWidthScaling=False
    OptionalParentItem=(Type=Bone,Name="None")
    RotationOffset=(X=40,Y=-40,Z=50)
   )
   > &RigUnit_ModifyTransforms(
    ItemToModify="((Item=(Type=Bone,Name=\"custom_upperarm_l\"),Transform=(Rotation=(X=0,Y=0,Z=-0.5,W=0.866),Translation=(X=-3,Y=0,Z=5),Scale3D=(X=1,Y=1.3,Z=1.1))),(Item=(Type=Bone,Name=\"dyn_L_armpitVolume\"),Transform=(Rotation=(X=0,Y=0.2588,Z=0,W=0.9659),Translation=(X=4,Y=0 ...(truncated)"
    WeightMinimum=0
    Weight=@n1.OutputParam
   )
   > n2 = &RigUnit_SphericalPoseReader(
    ActiveRegionScaleFactors=(PositiveWidth=1.5,NegativeWidth=7,PositiveHeight=3,NegativeHeight=3)
    ActiveRegionSize=0.07
    Debug=(bDrawDebug=false,bDraw2D=False,bDrawLocalAxes=False,DebugScale=25,DebugSegments=20,DebugThickness=0.25)
    DriverAxis=(X=-1,Y=0,Z=0)
    DriverItem=(Type=Bone,Name="upperarm_r")
    FalloffRegionScaleFactors=(PositiveWidth=1,NegativeWidth=1,PositiveHeight=1.5,NegativeHeight=1.5)
    FalloffSize=0.4
    FlipHeightScaling=false
    FlipWidthScaling=true
    OptionalParentItem=(Type=Bone,Name="None")
    RotationOffset=(X=40,Y=220,Z=230)
   )
   > &RigUnit_ModifyTransforms(
    ItemToModify="((Item=(Type=Bone,Name=\"custom_upperarm_r\"),Transform=(Rotation=(X=0,Y=0,Z=-0.5,W=0.866),Translation=(X=-3,Y=0,Z=-5),Scale3D=(X=1,Y=1.3,Z=1.1))),(Item=(Type=Bone,Name=\"dyn_R_armpitVolume\"),Transform=(Rotation=(X=0,Y=0.2588,Z=0,W=0.9659),Translation=(X=-4,Y ...(truncated)"
    WeightMinimum=0
    Weight=@n2.OutputParam
   )
   > &RigUnit_ParentConstraint(
    AdvancedSettings=(InterpolationType=Average,RotationOrderForFilter=XZY)
    Child=(Type=Bone,Name="dyn_L_scapulaAdjust")
    Parents=((Item=(Type=Bone,Name="spine_04"),Weight=0.5),(Item=(Type=Bone,Name="clavicle_l"),Weight=0.5))
    Weight=1
   )
   > &RigUnit_ParentConstraint(
    AdvancedSettings=(InterpolationType=Average,RotationOrderForFilter=XZY)
    Child=(Type=Bone,Name="dyn_R_scapulaAdjust")
    Parents=((Item=(Type=Bone,Name="spine_04"),Weight=0.5),(Item=(Type=Bone,Name="clavicle_r"),Weight=0.5))
    Weight=1
   )
   > &RigUnit_ParentConstraint(
    AdvancedSettings=(InterpolationType=Average,RotationOrderForFilter=XZY)
    Child=(Type=Bone,Name="dyn_L_aboveLeg_mass")
    Parents=((Item=(Type=Bone,Name="pelvis"),Weight=0.5),(Item=(Type=Bone,Name="thigh_l"),Weight=0.5))
    Weight=1
   )
   > &RigUnit_ParentConstraint(
    AdvancedSettings=(InterpolationType=Average,RotationOrderForFilter=XZY)
    Child=(Type=Bone,Name="dyn_R_aboveLeg_mass")
    Parents=((Item=(Type=Bone,Name="pelvis"),Weight=0.5),(Item=(Type=Bone,Name="thigh_r"),Weight=0.5))
    Weight=1
   )
   > &RigUnit_ParentConstraint(
    AdvancedSettings=(InterpolationType=Average,RotationOrderForFilter=XZY)
    Child=(Type=Bone,Name="dyn_L_belly")
    Parents=((Item=(Type=Bone,Name="pelvis"),Weight=0.3333),(Item=(Type=Bone,Name="spine_01"),Weight=0.3333),(Item=(Type=Bone,Name="spine_02"),Weight=0.3333))
    Weight=1
   )
   > &RigUnit_ParentConstraint(
    AdvancedSettings=(InterpolationType=Average,RotationOrderForFilter=XZY)
    Child=(Type=Bone,Name="dyn_C_belly")
    Parents=((Item=(Type=Bone,Name="pelvis"),Weight=0.3333),(Item=(Type=Bone,Name="spine_01"),Weight=0.3333),(Item=(Type=Bone,Name="spine_02"),Weight=0.3333))
    Weight=1
   )
   > &RigUnit_ParentConstraint(
    AdvancedSettings=(InterpolationType=Average,RotationOrderForFilter=XZY)
    Child=(Type=Bone,Name="dyn_R_belly")
    Parents=((Item=(Type=Bone,Name="pelvis"),Weight=0.3333),(Item=(Type=Bone,Name="spine_01"),Weight=0.3333),(Item=(Type=Bone,Name="spine_02"),Weight=0.3333))
    Weight=1
   )
   n3 = &RigUnit_GetRelativeTransformForItem(Child=(Type=Bone,Name="thigh_l"), Parent=(Type=Bone,Name="pelvis"), bChildInitial=False, bParentInitial=False)
   > &RigUnit_ModifyTransforms(
    ItemToModify=((Item=(Type=Bone,Name="dyn_L_aboveLeg_mass"),Transform=(Rotation=(X=0,Y=0,Z=0,W=-1),Translation=(X=0,Y=0,Z=6),Scale3D=(X=1,Y=1,Z=1))))
    WeightMinimum=0
    Weight=&RigVMFunction_MathDoubleRemap(
     SourceMaximum=-75
     SourceMinimum=-30
     TargetMaximum=1
     TargetMinimum=0
     Value=&RigVMFunction_MathQuaternionToEuler(RotationOrder=ZXY, Value=@n3.RelativeTransform.Rotation).Result.Z
    ).Result
   )
   Absolute = &RigVMFunction_MathDoubleAbs(Value=&RigVMFunction_MathQuaternionToEuler(
     RotationOrder=ZXY
     Value=&RigUnit_GetRelativeTransformForItem(Child=(Type=Bone,Name="thigh_r"), Parent=(Type=Bone,Name="pelvis"), bChildInitial=False, bParentInitial=False).RelativeTransform.Rotation
    ).Result.Z)
   Subtract = &RigVMFunction_MathDoubleSub(A=180, B=@Absolute.Result)
   > &RigUnit_ModifyTransforms(
    ItemToModify=((Item=(Type=Bone,Name="dyn_R_aboveLeg_mass"),Transform=(Rotation=(X=0,Y=0,Z=0,W=-1),Translation=(X=0,Y=0,Z=-6),Scale3D=(X=1,Y=1,Z=1))))
    WeightMinimum=0
    Weight=&RigVMFunction_MathDoubleRemap(SourceMaximum=75, SourceMinimum=30, TargetMaximum=1, TargetMinimum=0, Value=@Subtract.Result).Result
   )
   > n4 = &RigUnit_SphericalPoseReader(
    ActiveRegionScaleFactors=(PositiveWidth=1.5,NegativeWidth=1.5,PositiveHeight=1.5,NegativeHeight=1.5)
    ActiveRegionSize=0.12
    Debug=(bDrawDebug=false,DebugScale=25,DebugSegments=20,DebugThickness=0.25)
    DriverAxis=(X=1,Y=0,Z=0)
    DriverItem=(Type=Bone,Name="upperarm_l")
    FalloffRegionScaleFactors=(PositiveWidth=1,NegativeWidth=1,PositiveHeight=1,NegativeHeight=1)
    FalloffSize=0.6
    OptionalParentItem=(Type=Bone,Name="clavicle_l")
    RotationOffset=(X=50,Y=0,Z=0)
   )
   > &RigUnit_ModifyTransforms(
    ItemToModify=((Item=(Type=Bone,Name="upperarm_twist_01_l"),Transform=(Rotation=(X=0.2164,Y=0,Z=0,W=0.9763),Translation=(X=0,Y=0,Z=0),Scale3D=(X=1,Y=1,Z=1))))
    WeightMinimum=0
    Weight=@n4.OutputParam
   )
   > n5 = &RigUnit_SphericalPoseReader(
    ActiveRegionScaleFactors=(PositiveWidth=1.5,NegativeWidth=1.5,PositiveHeight=1.5,NegativeHeight=1.5)
    ActiveRegionSize=0.12
    Debug=(bDrawDebug=false,DebugScale=25,DebugSegments=20,DebugThickness=0.25)
    DriverAxis=(X=-1,Y=0,Z=0)
    DriverItem=(Type=Bone,Name="upperarm_r")
    FalloffRegionScaleFactors=(PositiveWidth=1,NegativeWidth=1,PositiveHeight=1,NegativeHeight=1)
    FalloffSize=0.6
    FlipHeightScaling=true
    FlipWidthScaling=false
    OptionalParentItem=(Type=Bone,Name="clavicle_r")
    RotationOffset=(X=50,Y=180,Z=180)
   )
   > &RigUnit_ModifyTransforms(
    ItemToModify=((Item=(Type=Bone,Name="upperarm_twist_01_r"),Transform=(Rotation=(X=0.2164,Y=0,Z=0,W=0.9763),Translation=(X=0,Y=0,Z=0),Scale3D=(X=1,Y=1,Z=1))))
    WeightMinimum=0
    Weight=@n5.OutputParam
   )
  ]}
]
hierarchy [
 root {loc=<0 0 0>, rot=<0 0 0 1>}
 pelvis {parent=root, loc=<0 -6.4479 73.7286>, rot=<0.0127 0.707 -0.0127 -0.707>}
 spine_01 {parent=pelvis, loc=<7.0662 -0.3925 0>, rot=<0 0 0.034 -0.9994>}
 spine_02 {parent=spine_01, loc=<11.1087 0 0>, rot=<0 0 0.0238 -0.9997>}
 spine_03 {parent=spine_02, loc=<11.1087 0 0>, rot=<0 0 0.0149 -0.9999>}
 spine_04 {parent=spine_03, loc=<11.1087 0 0>, rot=<0 0 -0.0132 -0.9999>}
 spine_05 {parent=spine_04, loc=<11.1087 0 0>, rot=<0 0 -0.0015 -1>}
 neck_01 {parent=spine_05, loc=<11.8792 7.6258 0>, rot=<0 0 -0.2003 -0.9797>}
 neck_02 {parent=neck_01, loc=<4.7445 0 0>, rot=<0 0 0.0504 -0.9987>}
 head {parent=neck_02, loc=<5.7899 0 0>, rot=<0 0 0.115 -0.9934>}
 Hat {parent=head, type=Null, loc=<0 0 0>, rot=<0 0.7071 0 0.7071>}
 FX_Head {parent=head, type=Null, loc=<0 0 0>, rot=<0 0 0 1>}
 Glasses {parent=head, type=Null, loc=<9.0729 6.6834 0>, rot=<0 0.7071 0 0.7071>}
 faceAttach {parent=head, loc=<0.2056 -1.9691 0>, rot=<0.0029 -0.7071 -0.0029 -0.7071>}
 C_brow_mid {parent=faceAttach, loc=<0 15.9166 15.8438>, rot=<0 0 0 -1>}
 L_brow_mid {parent=faceAttach, loc=<4.4665 15.6723 16.0616>, rot=<-0.0004 0.0188 0.0279 -0.9994>}
 L_brow_outer {parent=faceAttach, loc=<7.8005 15.1681 15.887>, rot=<-0.097 -0.0533 0.1811 -0.9772>}
 L_cheek_inner {parent=faceAttach, loc=<14.3175 18.707 5>, rot=<-0.21 -0.1846 0.2019 -0.9387>}
 C_lip_upper_mid {parent=faceAttach, loc=<-0.0324 23.7608 2.7715>, rot=<-0.0489 0.0011 -0.0008 -0.9988>}
 L_lip_upper_outer {parent=faceAttach, loc=<4.5473 22.6469 2.8185>, rot=<0.0592 0.0132 0.2585 -0.9641>}
 L_lip_corner {parent=faceAttach, loc=<7.9421 20.2812 2.7743>, rot=<0.0007 0.0085 0.2588 -0.9659>}
 C_nose_bridge {parent=faceAttach, loc=<0 18.8775 10.5912>, rot=<-0.7679 0 0 -0.6405>}
 R_brow_mid {parent=faceAttach, loc=<-4.4665 15.6723 16.0616>, rot=<-0.0004 -0.0188 -0.0279 -0.9994>}
 R_brow_outer {parent=faceAttach, loc=<-7.8005 15.1681 15.887>, rot=<-0.097 0.0533 -0.1811 -0.9772>}
 R_cheek_inner {parent=faceAttach, loc=<-14.3175 18.707 5>, rot=<-0.21 0.1846 -0.2019 -0.9387>}
 R_lip_corner {parent=faceAttach, loc=<-7.9421 20.2812 2.7743>, rot=<0.0007 -0.0085 -0.2588 -0.9659>}
 C_jaw {parent=faceAttach, loc=<-0.0134 1.8686 4.5384>, rot=<0 0 0 1>}
 L_lip_lower_outer {parent=C_jaw, loc=<5.1412 21.5121 -3.1092>, rot=<-0.3576 0.1023 0.2885 -0.8823>}
 C_lip_lower_mid {parent=C_jaw, loc=<0.0223 22.5116 -2.6322>, rot=<-0.2986 0 0 -0.9544>}
 R_lip_lower_outer {parent=C_jaw, loc=<-5.1144 21.5121 -3.1092>, rot=<-0.3576 -0.1023 -0.2885 -0.8823>}
 teeth_lower {parent=C_jaw, loc=<0.0134 13.1571 -3.4583>, rot=<0 0 0 1>}
 tongue {parent=C_jaw, loc=<0.0141 7.7112 -2.7212>, rot=<-0.375 0 0 -0.927>}
 teeth_upper {parent=faceAttach, loc=<0 15.0292 2.589>, rot=<0 0 0 1>}
 R_lip_upper_outer {parent=faceAttach, loc=<-4.5473 22.6469 2.8185>, rot=<0.0592 -0.0132 -0.2585 -0.9641>}
 SM_Helmet {parent=neck_02, type=Null, loc=<-13.2064 1.1483 0>, rot=<0 0.7071 0 0.7071>}
 attach_cape {parent=spine_05, loc=<9.017 -11.714 0>, rot=<0.0616 -0.7044 -0.0616 -0.7044>}
 cape {parent=attach_cape, type=Null, loc=<-0.1533 -1.8982 -0.1791>, rot=<0.0471 0.0019 -0.0401 0.9981>}
 attach_backpack {parent=spine_05, loc=<-0.123 -13.473 0>, rot=<-0.0093 -0.707 0.0093 -0.707>}
 Backpack_BR {parent=attach_backpack, type=Null, loc=<0 0.0003 0.0002>, rot=<-0.0131 0 0 0.9999>}
 clavicle_l {parent=spine_05, loc=<4.5642 5.0167 -7.5493>, rot=<0.0465 -0.7289 -0.009 -0.6829>}
 upperarm_l {parent=clavicle_l, loc=<15.0437 0.1342 -0.001>, rot=<0.0226 -0.3234 -0.0088 -0.9459>}
 lowerarm_l {parent=upperarm_l, loc=<26.9311 0.1343 -0.7608>, rot=<0.0391 -0.0099 -0.2453 -0.9686>}
 hand_l {parent=lowerarm_l, loc=<26.5233 -0.4252 1.1978>, rot=<-0.7081 -0.0507 0.0975 -0.6975>}
 thumb_01_l {parent=hand_l, loc=<2.356 -1.7549 -5.1281>, rot=<0.6546 -0.4042 0.0664 -0.6354>}
 thumb_02_l {parent=thumb_01_l, loc=<4.8027 0 0>, rot=<-0.0037 0.0213 0.298 -0.9543>}
 thumb_03_l {parent=thumb_02_l, loc=<3.9275 0 0>, rot=<0.0118 -0.013 0.0066 -0.9998>}
 index_metacarpal_l {parent=hand_l, loc=<1.2106 1.1713 -4.8194>, rot=<0.0158 -0.0124 -0.0173 -0.9996>}
 index_01_l {parent=index_metacarpal_l, loc=<7.9817 0 0>, rot=<-0.0044 0.0485 0.0924 -0.9945>}
 index_02_l {parent=index_01_l, loc=<4.0632 0 0>, rot=<-0.002 0.0027 0.3225 -0.9466>}
 index_03_l {parent=index_02_l, loc=<3.5435 0 0>, rot=<-0.002 0.0067 0.2558 -0.9667>}
 middle_metacarpal_l {parent=hand_l, loc=<1.8652 0.8261 -0.0739>, rot=<0.006 0.0168 -0.0329 -0.9993>}
 middle_01_l {parent=middle_metacarpal_l, loc=<7.1989 0.1082 -0.039>, rot=<-0.0208 0.0042 0.0858 -0.9961>}
 middle_02_l {parent=middle_01_l, loc=<3.859 0 0>, rot=<-0.0218 0.0258 0.3512 -0.9357>}
 middle_03_l {parent=middle_02_l, loc=<3.5809 0 0>, rot=<0.0082 0.0188 0.1708 -0.9851>}
 ring_metacarpal_l {parent=hand_l, loc=<1.5368 0.3901 3.9044>, rot=<-0.0284 0.0722 -0.0458 -0.9959>}
 ring_01_l {parent=ring_metacarpal_l, loc=<5.7881 -0.753 -0.0969>, rot=<-0.0489 -0.0252 0.1892 -0.9804>}
 ring_02_l {parent=ring_01_l, loc=<4.2347 0 0>, rot=<0 0 0.2806 -0.9598>}
 ring_03_l {parent=ring_02_l, loc=<2.7257 0 0>, rot=<0 0 0.1189 -0.9929>}
 hand_pickaxe_l {parent=hand_l, type=Null, loc=<6.8124 0 0>, rot=<0.7071 0 0.7071 0>}
 FX_Snap {parent=hand_l, type=Null, loc=<13.4037 -12.5251 -2.3306>, rot=<-0.5914 0.3535 -0.4031 0.6023>}
 pinky_metacarpal_l {parent=hand_l, loc=<0 0 6.745>, rot=<-0.0397 0.0814 -0.069 -0.9935>}
 pinky_01_l {parent=pinky_metacarpal_l, loc=<5.9107 -0.7689 -0.099>, rot=<-0.0654 -0.0035 0.2242 -0.9723>}
 pinky_02_l {parent=pinky_01_l, loc=<4.2347 0 0>, rot=<0.0037 0.0011 0.2854 -0.9584>}
 pinky_03_l {parent=pinky_02_l, loc=<2.7257 0 0>, rot=<-0.0002 0.0098 0.1445 -0.9895>}
 weapon_l {parent=hand_l, loc=<0 -3.411 0>, rot=<0 0 0 1>}
 FX_SoulSuck_L {parent=weapon_l, type=Null, loc=<0 5.122 0>, rot=<-0.7071 0 0 0.7071>}
 Melee_LSword_Upper {parent=weapon_l, type=Null, loc=<3.2939 -17.9222 97.3085>, rot=<0 0 0 1>}
 Melee_LSword_Lower {parent=weapon_l, type=Null, loc=<1.3975 -7.6031 41.2824>, rot=<0 0 0 1>}
 Melee_L_Lower {parent=weapon_l, type=Null, loc=<-1.1787 6.42 -34.8603>, rot=<0 0 0 1>}
 Melee_L_Upper {parent=weapon_l, type=Null, loc=<-3.5614 19.3781 -105.2158>, rot=<0 0 0 1>}
 LeftHand {parent=weapon_l, type=Null, loc=<0 0 0>, rot=<0 0 -0.7071 0.7071>}
 BluePrintPaper {parent=weapon_l, type=Null, loc=<0 0 0>, rot=<0 0 0 1>}
 hand_attach_l {parent=hand_l, loc=<8.5 -3 0>, rot=<0.7071 0.7071 0 0>}
 lowerarm_twist_02_l {parent=lowerarm_l, loc=<8.6187 0 0>, rot=<0 0 0 1>}
 lowerarm_twist_01_l {parent=lowerarm_l, loc=<17.2375 0 0>, rot=<0 0 0 1>}
 deform_elbow_in_base_l {parent=lowerarm_l, loc=<1.8474 4.6137 0.059>, rot=<0.0257 0.0489 0.0163 -0.9983>}
 deform_elbow_in_l {parent=deform_elbow_in_base_l, loc=<0 0 0>, rot=<0 0 0 1>}
 deform_elbow_out_base_l {parent=lowerarm_l, loc=<0.0992 -5.2741 -0.4171>, rot=<0.0257 0.0489 0.0163 -0.9983>}
 deform_elbow_out_l {parent=deform_elbow_out_base_l, loc=<0 0 0>, rot=<0 0 0 1>}
 wrist_pickaxe_l {parent=lowerarm_l, type=Null, loc=<35.6995 1.1516 -0.5758>, rot=<0 -0.7934 -0.6088 0>}
 lowerarm_weapon_l {parent=lowerarm_l, type=Null, loc=<10 0 0>, rot=<-0.5 0.5 -0.5 0.5>}
 FX_Contrail_Hand_L {parent=lowerarm_l, type=Null, loc=<27.1033 -0.0003 0>, rot=<-1 0 0 0>}
 ConstructorShield {parent=lowerarm_l, type=Null, loc=<20.8488 -0.0002 0>, rot=<-0.3196 -0.3196 -0.6308 0.6308>}
 SM_L_ArmFore {parent=lowerarm_l, type=Null, loc=<0 0 0>, rot=<-0.4789 -0.2712 0.835 0>}
 upperarm_twist_01_l {parent=upperarm_l, loc=<8.968 0.0447 -0.2534>, rot=<0 0 0 1>}
 upperarm_twist_02_l {parent=upperarm_l, loc=<17.9361 0.0894 -0.5067>, rot=<0 0 0 1>}
 SM_L_Shoulder {parent=upperarm_l, type=Null, loc=<1.2171 -0.1862 1.5757>, rot=<-0.4762 -0.0286 0.8786 0.0225>}
 custom_upperarm_l {parent=upperarm_l, loc=<-0.0984 -0.0072 0.2485>, rot=<0 0 0 1>}
 custom_clavicle_l {parent=clavicle_l, loc=<-0.023 0 0.3427>, rot=<0 0 0 1>}
 clavicle_r {parent=spine_05, loc=<4.5646 5.0167 7.5493>, rot=<0.7289 0.0465 -0.6829 0.009>}
 upperarm_r {parent=clavicle_r, loc=<-15.0437 -0.1342 0.0018>, rot=<0.0226 -0.3234 -0.0088 -0.9459>}
 lowerarm_r {parent=upperarm_r, loc=<-26.9309 -0.1343 0.7605>, rot=<0.0391 -0.0099 -0.2453 -0.9686>}
 hand_r {parent=lowerarm_r, loc=<-26.5233 0.4252 -1.1978>, rot=<-0.7081 -0.0507 0.0975 -0.6975>}
 thumb_01_r {parent=hand_r, loc=<-2.3559 1.7549 5.1281>, rot=<0.6546 -0.4042 0.0664 -0.6354>}
 thumb_02_r {parent=thumb_01_r, loc=<-4.8027 0.0001 0>, rot=<-0.0037 0.0213 0.298 -0.9543>}
 thumb_03_r {parent=thumb_02_r, loc=<-3.9275 0 0>, rot=<0.0118 -0.013 0.0066 -0.9998>}
 index_metacarpal_r {parent=hand_r, loc=<-1.2101 -1.1715 4.8195>, rot=<0.0158 -0.0124 -0.0173 -0.9996>}
 index_01_r {parent=index_metacarpal_r, loc=<-7.9821 0.0003 -0.0001>, rot=<-0.0044 0.0485 0.0924 -0.9945>}
 index_02_r {parent=index_01_r, loc=<-4.0632 0 0>, rot=<-0.002 0.0027 0.3225 -0.9466>}
 index_03_r {parent=index_02_r, loc=<-3.5436 0 0>, rot=<-0.002 0.0067 0.2558 -0.9667>}
 middle_metacarpal_r {parent=hand_r, loc=<-1.8652 -0.8261 0.0739>, rot=<0.006 0.0168 -0.0329 -0.9993>}
 middle_01_r {parent=middle_metacarpal_r, loc=<-7.1989 -0.1081 0.0391>, rot=<-0.0208 0.0042 0.0858 -0.9961>}
 middle_02_r {parent=middle_01_r, loc=<-3.8589 0 0>, rot=<-0.0217 0.0259 0.3512 -0.9357>}
 middle_03_r {parent=middle_02_r, loc=<-3.5809 0 0>, rot=<0.0083 0.0188 0.1708 -0.9851>}
 FX_Candy {parent=middle_01_r, type=Null, loc=<0 0 0>, rot=<0.0534 0.2587 0.7482 0.6086>}
 ring_metacarpal_r {parent=hand_r, loc=<-1.5368 -0.3901 -3.9044>, rot=<-0.0284 0.0722 -0.0458 -0.9959>}
 ring_01_r {parent=ring_metacarpal_r, loc=<-5.7881 0.7529 0.0969>, rot=<-0.0489 -0.0252 0.1892 -0.9804>}
 ring_02_r {parent=ring_01_r, loc=<-4.2347 0.0001 0>, rot=<0 0 0.2806 -0.9598>}
 ring_03_r {parent=ring_02_r, loc=<-2.7258 0 0>, rot=<0 0 0.1189 -0.9929>}
 hand_pickaxe_r {parent=hand_r, type=Null, loc=<-6.8124 0 0>, rot=<0 -0.7071 0 0.7071>}
 FX_EmoteSalt {parent=hand_r, type=Null, loc=<-11.3763 -0.0051 3.0155>, rot=<-0.5048 0.5826 -0.2599 0.5816>}
 weapon_r_socket {parent=hand_r, type=Null, loc=<0.3838 6.63 -0.7407>, rot=<0 0 0 1>}
 pinky_metacarpal_r {parent=hand_r, loc=<0 0 -6.745>, rot=<-0.0397 0.0814 -0.069 -0.9935>}
 pinky_01_r {parent=pinky_metacarpal_r, loc=<-5.9107 0.7689 0.099>, rot=<-0.0654 -0.0035 0.2242 -0.9723>}
 pinky_02_r {parent=pinky_01_r, loc=<-4.2347 0.0001 0>, rot=<0.0037 0.0011 0.2854 -0.9584>}
 pinky_03_r {parent=pinky_02_r, loc=<-2.7258 0 0>, rot=<-0.0002 0.0098 0.1445 -0.9895>}
 weapon_r {parent=hand_r, loc=<0 3.411 0>, rot=<0 0 0 1>}
 Melee_IBlade_Upper {parent=weapon_r, type=Null, loc=<4.6122 -25.1006 136.2771>, rot=<0 0 0 1>}
 Melee_IBlade_Lower {parent=weapon_r, type=Null, loc=<1.397 -7.603 41.2825>, rot=<0 0 0 1>}
 FX_Melee_R_Speedy1 {parent=weapon_r, type=Null, loc=<0.8984 -4.8878 26.5387>, rot=<0 0 0 1>}
 FX_Melee_R_Speedy {parent=weapon_r, type=Null, loc=<2.2293 -12.1289 65.8553>, rot=<0 0 0 1>}
 FX_Melee_R_Carbide1 {parent=weapon_r, type=Null, loc=<5.5961 1.5957 41.3084>, rot=<0 0 0 1>}
 FX_Melee_R_Carbide {parent=weapon_r, type=Null, loc=<-2.901 -16.2591 38.3077>, rot=<0 0 0 1>}
 Melee_R_UpperRight {parent=weapon_r, type=Null, loc=<-9.8403 2.5372 56.7564>, rot=<0 0 0.8191 0.5737>}
 Melee_R_UpperLeft {parent=weapon_r, type=Null, loc=<-15.4484 -9.247 54.776>, rot=<0 0 0.11 0.9939>}
 FX_SoulSuck_R {parent=weapon_r, type=Null, loc=<0.0001 7.6472 -0.0005>, rot=<-0.7071 0 0 0.7071>}
 Melee_RSword_Upper {parent=weapon_r, type=Null, loc=<3.2939 -17.9219 97.3086>, rot=<0 0 0 1>}
 Melee_RSword_Lower {parent=weapon_r, type=Null, loc=<1.397 -7.603 41.2825>, rot=<0 0 0 1>}
 Melee_R_Lower {parent=weapon_r, type=Null, loc=<-0.4991 2.7154 -14.7437>, rot=<0 0 0 1>}
 Melee_R_Upper {parent=weapon_r, type=Null, loc=<1.7302 -9.4135 51.1116>, rot=<0 0 0 1>}
 BluePrintPen {parent=weapon_r, type=Null, loc=<0 0 0>, rot=<0 0 0 1>}
 RightHand {parent=weapon_r, type=Null, loc=<0 0 0>, rot=<0 0 -0.7071 0.7071>}
 hand_attach_r {parent=hand_r, loc=<-8.5 3 0>, rot=<0 0 -0.7071 -0.7071>}
 lowerarm_twist_02_r {parent=lowerarm_r, loc=<-8.6185 -0.0001 -0.0002>, rot=<0 0 0 1>}
 lowerarm_twist_01_r {parent=lowerarm_r, loc=<-17.2376 0.0001 0.0003>, rot=<0 0 0 1>}
 deform_elbow_in_base_r {parent=lowerarm_r, loc=<-1.8475 -4.6137 -0.0589>, rot=<0.0257 0.0489 0.0163 -0.9983>}
 deform_elbow_in_r {parent=deform_elbow_in_base_r, loc=<0 0 0>, rot=<0 0 0 1>}
 deform_elbow_out_base_r {parent=lowerarm_r, loc=<-0.099 5.274 0.417>, rot=<0.0257 0.0489 0.0163 -0.9983>}
 deform_elbow_out_r {parent=deform_elbow_out_base_r, loc=<0 0 0>, rot=<0 0 0 1>}
 wrist_pickaxe_r {parent=lowerarm_r, type=Null, loc=<-35.6995 -1.1516 0.5758>, rot=<0.6088 0 0 0.7934>}
 lowerarm_weapon_r {parent=lowerarm_r, type=Null, loc=<-10 0 0>, rot=<0.5 -0.5 -0.5 0.5>}
 FX_Contrail_Hand_R {parent=lowerarm_r, type=Null, loc=<-25.7482 0.0007 -0.0001>, rot=<0 -1 0 0>}
 FX_HandBeam_R {parent=lowerarm_r, type=Null, loc=<-46.773 0 0>, rot=<0 -1 0 0>}
 SM_R_ArmFore {parent=lowerarm_r, type=Null, loc=<-0.7922 4.8355 3.4504>, rot=<-0.8596 0.2166 -0.4548 0.086>}
 upperarm_twist_01_r {parent=upperarm_r, loc=<-8.9676 -0.0447 0.2528>, rot=<0 0 0 1>}
 upperarm_twist_02_r {parent=upperarm_r, loc=<-17.9358 -0.0894 0.5063>, rot=<0 0 0 1>}
 SM_R_Shoulders {parent=upperarm_r, type=Null, loc=<-3.8771 0.0002 6.4605>, rot=<-0.8777 0.0864 -0.4691 0.0462>}
 custom_upperarm_r {parent=upperarm_r, loc=<0.0988 0.0073 -0.249>, rot=<0 0 0 1>}
 custom_clavicle_r {parent=clavicle_r, loc=<0.023 0 -0.3427>, rot=<0 0 0 1>}
 custom_neck_01 {parent=spine_05, loc=<11.8792 7.6258 0>, rot=<0 0 -0.2003 -0.9797>}
 custom_neck_02 {parent=custom_neck_01, loc=<4.7445 0 0>, rot=<0 0 0.0504 -0.9987>}
 custom_head {parent=custom_neck_02, loc=<5.7899 0 0>, rot=<0 0 0.115 -0.9934>}
 Backpack {parent=spine_05, type=Null, loc=<0 0 0>, rot=<0 0.7071 0 0.7071>}
 FX_TechStack {parent=spine_05, type=Null, loc=<55.1518 -14.5915 -14.005>, rot=<0 0 0 1>}
 FX_Chest {parent=spine_05, type=Null, loc=<0 0 0>, rot=<0.4995 0.4996 0.5004 0.5005>}
 dyn_L_scapulaAdjust {parent=spine_04, loc=<9.9546 -11.1794 -13.4829>, rot=<0.426 -0.5525 0.5936 -0.4012>}
 dyn_R_scapulaAdjust {parent=spine_04, loc=<9.9547 -11.1793 13.4829>, rot=<0.5525 0.426 -0.4012 -0.5936>}
 LootEntryPosition {parent=spine_04, type=Null, loc=<20.438 -18.9584 0>, rot=<0 0 0 1>}
 FX_TechClip {parent=spine_04, type=Null, loc=<1.2692 34.2335 44.2145>, rot=<0.4583 0.5385 0.5 0.5>}
 Shield {parent=spine_03, type=Null, loc=<31.1437 65.733 -17.714>, rot=<0.4056 0.5792 0.4056 0.5792>}
 FX_Spine {parent=spine_03, type=Null, loc=<0 0 0>, rot=<0.4996 0.4995 0.5005 0.5004>}
 dyn_L_armpitVolume {parent=spine_02, loc=<15.2578 4.9724 -27.0935>, rot=<0.0255 -0.6398 -0.0306 -0.7675>}
 dyn_R_armpitVolume {parent=spine_02, loc=<15.258 4.9724 27.0935>, rot=<-0.6398 -0.0255 0.7675 -0.0306>}
 thigh_l {parent=pelvis, loc=<-12.9663 4.6874 -15.2886>, rot=<0.0012 0.0572 0.0155 -0.9982>}
 calf_l {parent=thigh_l, loc=<-29.2681 0.033 0.3137>, rot=<0.0006 -0.0161 -0.0391 -0.9991>}
 calf_twist_02_l {parent=calf_l, loc=<-5 0 0>, rot=<0.0237 0.0095 -0.0035 -0.9997>}
 calf_twist_01_l {parent=calf_l, loc=<-13 0 0>, rot=<0.0237 0.0095 -0.0035 -0.9997>}
 foot_l {parent=calf_l, loc=<-19.7455 0 0>, rot=<0.0418 -0.0455 0.0343 -0.9975>}
 ball_l {parent=foot_l, loc=<-6.225 14.5256 0.2872>, rot=<0.0001 -0.0001 -0.7063 -0.7079>}
 Toe_L {parent=ball_l, type=Null, loc=<0.0436 5.2798 -0.0251>, rot=<0 0 0 1>}
 FX_Skydive_Foot_L {parent=foot_l, type=Null, loc=<0 0 0>, rot=<0 0 0 1>}
 foot_l {parent=foot_l, type=Null, loc=<-11.8074 0.1251 -0.0562>, rot=<0 0 0 1>}
 FX_foot_lSocket {parent=foot_l, type=Null, loc=<-10.7571 -16.1416 0>, rot=<0 0 0 1>}
 deform_knee_in_base_l {parent=calf_l, loc=<0 -6.8431 0>, rot=<0 0 0 1>}
 deform_knee_in_l {parent=deform_knee_in_base_l, loc=<0 0 0>, rot=<0 0 0 1>}
 deform_knee_out_base_l {parent=calf_l, loc=<0 8.0664 0>, rot=<0 0 0 1>}
 deform_knee_out_l {parent=deform_knee_out_base_l, loc=<0 0 0>, rot=<0 0 0 1>}
 FX_Contrail_Foot_L {parent=calf_l, type=Null, loc=<-23.159 -0.0001 0>, rot=<0 -0.7071 0.7071 0>}
 SM_L_Shin {parent=calf_l, type=Null, loc=<0 0 0>, rot=<0.0295 0.7074 0.0017 0.7062>}
 thigh_twist_01_l {parent=thigh_l, loc=<-12 0 0>, rot=<0.0013 0 0 -1>}
 SM_L_Thigh {parent=thigh_l, type=Null, loc=<0 0 0>, rot=<0.0674 0.7029 0.0367 0.7071>}
 thigh_r {parent=pelvis, loc=<-12.9663 4.6874 15.2886>, rot=<0.0572 -0.0012 0.9982 0.0155>}
 calf_r {parent=thigh_r, loc=<29.2682 -0.033 -0.3137>, rot=<0.0006 -0.0161 -0.0391 -0.9991>}
 calf_twist_02_r {parent=calf_r, loc=<5 0 0.0001>, rot=<0.0237 0.0095 -0.0035 -0.9997>}
 calf_twist_01_r {parent=calf_r, loc=<13 0 0>, rot=<0.0237 0.0095 -0.0035 -0.9997>}
 foot_r {parent=calf_r, loc=<19.7454 0 0>, rot=<0.0418 -0.0455 0.0343 -0.9975>}
 ball_r {parent=foot_r, loc=<6.225 -14.5256 -0.2872>, rot=<0.0001 -0.0001 -0.7063 -0.7079>}
 Toe_R {parent=ball_r, type=Null, loc=<-0.0436 -5.28 0.0251>, rot=<0 0 0 1>}
 FX_Skydive_Foot_R {parent=foot_r, type=Null, loc=<0 0 0>, rot=<0 0 0 1>}
 foot_r {parent=foot_r, type=Null, loc=<11.8076 -0.1251 0.0562>, rot=<0 0 0 1>}
 FX_foot_rSocket {parent=foot_r, type=Null, loc=<10.7571 16.1416 0>, rot=<0 0 1 0>}
 deform_knee_in_base_r {parent=calf_r, loc=<-0.0001 6.8431 0>, rot=<0 0 0 1>}
 deform_knee_in_r {parent=deform_knee_in_base_r, loc=<0 0 0>, rot=<0 0 0 1>}
 deform_knee_out_base_r {parent=calf_r, loc=<0 -8.0664 0>, rot=<0 0 0 1>}
 deform_knee_out_r {parent=deform_knee_out_base_r, loc=<0 0 0>, rot=<0 0 0 1>}
 FX_Contrail_Foot_R {parent=calf_r, type=Null, loc=<23.2886 0.0007 0.0004>, rot=<0.7071 0 0 0.7071>}
 SM_R_Shin {parent=calf_r, type=Null, loc=<-0.0515 -0.5623 -0.0288>, rot=<0.7 -0.0276 -0.7136 0>}
 thigh_twist_01_r {parent=thigh_r, loc=<12 0 -0.0001>, rot=<0.0013 0 0 -1>}
 SM_R_Thigh {parent=thigh_r, type=Null, loc=<0 0 0>, rot=<0.7059 -0.0308 -0.7077 0>}
 deform_glute_base_l {parent=pelvis, loc=<-11.8186 -7.9325 -11.6362>, rot=<0.0012 0.0572 0.0155 -0.9982>}
 deform_glute_l {parent=deform_glute_base_l, loc=<0 0 0>, rot=<0 0 0 1>}
 deform_glute_base_r {parent=pelvis, loc=<-11.8186 -7.9325 11.6362>, rot=<0.0572 -0.0012 0.9982 0.0155>}
 deform_glute_r {parent=deform_glute_base_r, loc=<0 0 0>, rot=<0 0 0 -1>}
 dyn_C_belly {parent=pelvis, loc=<9.4488 31.2165 0>, rot=<-0.5185 -0.4808 -0.4808 -0.5185>}
 dyn_L_belly {parent=pelvis, loc=<10.7823 20.3514 -24.131>, rot=<-0.2657 -0.6303 -0.2123 -0.6978>}
 dyn_R_belly {parent=pelvis, loc=<10.7823 20.3514 24.131>, rot=<-0.6303 0.2657 0.6978 -0.2123>}
 dyn_L_aboveLeg_mass {parent=pelvis, loc=<-17.9928 13.8934 -17.0623>, rot=<-0.4656 -0.5348 -0.4662 -0.529>}
 dyn_R_aboveLeg_mass {parent=pelvis, loc=<-17.9928 13.8934 17.0623>, rot=<0.5348 -0.4656 -0.529 0.4662>}
 dyn_beltBuckle {parent=pelvis, loc=<-8.6219 39.8763 -0.0002>, rot=<-0.0127 -0.707 0.0127 -0.707>}
 FX_Combat {parent=root, type=Null, loc=<0 -0.0002 188.3311>, rot=<0 0 0.7071 0.7071>}
 Root_Effect {parent=root, type=Null, loc=<0 0 0>, rot=<0 0 0.7071 0.7071>}
 root_fp {parent=root, loc=<0 0 0>, rot=<0 0 0 1>}
 camera_root {parent=root_fp, loc=<0 0 0>, rot=<0 0 0 1>}
 camera {parent=camera_root, loc=<0 0 0>, rot=<0 0 0 1>}
 attach_fp {parent=root_fp, loc=<0 0 0>, rot=<0 0 0 1>}
 warp_point_01 {parent=root, loc=<0 0 0>, rot=<0 0 0 1>}
 warp_point_02 {parent=root, loc=<0 0 0>, rot=<0 0 0 1>}
 warp_point_03 {parent=root, loc=<0 0 0>, rot=<0 0 0 1>}
 attach {parent=root, loc=<0 0 0>, rot=<0 0 0 1>}
 attach_emote_01 {parent=root, loc=<0 0 0>, rot=<0 0 0 1>}
 attach_emote_02 {parent=root, loc=<0 0 0>, rot=<0 0 0 1>}
 attach_emote_03 {parent=root, loc=<0 0 0>, rot=<0 0 0 1>}
 ik_foot_root {parent=root, loc=<0 0 0>, rot=<0 0 0 1>}
 ik_foot_l {parent=ik_foot_root, loc=<19.9326 -3.9733 11.8791>, rot=<0.0357 0.7045 0.0284 -0.7082>}
 ik_foot_r {parent=ik_foot_root, loc=<-19.9325 -3.9733 11.8791>, rot=<0.7082 0.0284 -0.7045 0.0357>}
 ik_hand_root {parent=root, loc=<0 0 0>, rot=<0 0 0 1>}
 ik_hand_gun {parent=ik_hand_root, loc=<-60.0749 4.6707 94.6826>, rot=<0.6084 0.1793 0.4613 -0.6205>}
 ik_hand_l {parent=ik_hand_gun, loc=<61.2952 94.9889 40.6933>, rot=<-0.5102 -0.7906 -0.3387 0>}
 ik_hand_r {parent=ik_hand_gun, loc=<0 0 0>, rot=<0 0 0 1>}
]
