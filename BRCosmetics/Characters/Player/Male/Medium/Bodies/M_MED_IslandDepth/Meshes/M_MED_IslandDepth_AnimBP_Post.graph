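# Post-process AnimBlueprint for the IslandDepth medium male body, adding secondary motion to the belly and belt buckle.
# The pose runs through a Control Rig, a weighted belt buckle constraint and angular-spring AnimDynamics on three belly bones.
# A following offset and linear-spring belly pass has inverted alpha scale/bias, apparently zeroing its effect.
# EventGraph computes a neck_02 counter-rotation, but its bone node and the neckFollow constraint are unconnected to the output.
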AnimBlueprint /BRCosmetics/Characters/Player/Male/Medium/Bodies/M_MED_IslandDepth/Meshes/M_MED_IslandDepth_AnimBP_Post
target_skeleton=&/M_MED_IslandDepth_Skeleton
graphs [
 AnimGraph {nodes [
   "Constraint - dyn_neckFollow (P)"(Alpha=1, ConstraintWeights_0=1) {boneToModify {boneName=dyn_neckFollow}, constraintSetup [{targetBone {boneName=custom_head}}], constraintWeights [1]}
   "Transform (Modify) Bone - Bone: neck_02"(Alpha=1, Translation="0,0,0", Rotation=Get neck02 rotate() -> (neck02 rotate:Rotator), Scale="1,1,1") {rotationMode=BMM_Additive, rotationSpace=BCS_BoneSpace, boneToModify {boneName=neck_02}}
   n1 = Control Rig(Source=Input Pose()) {controlRigAssetReference {blueprintRigClass=&/M_MED_IslandDepth_CtrlRig.M_MED_IslandDepth_CtrlRig_C}}
   n2 = "Constraint - dyn_beltBuckle (PPPPPPPP)"!(
    ComponentPose=Local To Component(LocalPose=@n1)
    Alpha=1
    ConstraintWeights_0=0.76
    ConstraintWeights_1:float=0.1246
    ConstraintWeights_2:float=0.0431
    ConstraintWeights_3:float=0.0255
    ConstraintWeights_4:float=0.0224
    ConstraintWeights_5:float=0.022
    ConstraintWeights_6:float=0.0014
    ConstraintWeights_7:float=0.0011
   ) -> (Pose) {
    boneToModify {boneName=dyn_beltBuckle}
    constraintSetup [
     {targetBone {boneName=pelvis}}
     {targetBone {boneName=dyn_C_belly}}
     {targetBone {boneName=dyn_L_aboveLeg_mass}}
     {targetBone {boneName=dyn_R_aboveLeg_mass}}
     {targetBone {boneName=dyn_L_belly}}
     {targetBone {boneName=dyn_R_belly}}
     {targetBone {boneName=spine_04}}
     {targetBone {boneName=spine_01}}
    ]
    constraintWeights [1,1,1,1,1,1,1,1]
   }
   n3 = Anim Dynamics - Bone: dyn_L_belly(ComponentPose=@n2, Alpha=1, ExternalForce="0,0,0") {
    angularDampingOverride=1
    angularSpringConstant=15000
    bAngularSpring=true
    bOverrideAngularDamping=true
    bUseGravityOverride=true
    gravityScale=0.35
    simulationSpace=RootRelative
    boundBone {boneName=dyn_L_belly}
    componentAppliedLinearAccClamp=<3000 3000 3000>
    componentLinearAccScale=<1 1 1>
    componentLinearVelScale=<1 1 1>
    gravityOverride {z=-980}
    physicsBodyDefinitions [{
      boundBone {boneName=dyn_L_belly}
      boxExtents=<28 28 28>
      constraintSetup {angularLimitsMax=<10 15 20>, angularLimitsMin=<-10 -15 -20>, angularTarget=<1 0 0>}
      localJointOffset=<3 0 0>
     }]
    simSpaceSettings {maxAngularAcceleration=25, maxAngularVelocity=5, simSpaceAngularAlpha=1}
   }
   n4 = Anim Dynamics - Bone: dyn_C_belly(ComponentPose=@n3, Alpha=1, ExternalForce="0,0,0") {
    angularDampingOverride=1
    angularSpringConstant=15000
    bAngularSpring=true
    bOverrideAngularDamping=true
    bUseGravityOverride=true
    gravityScale=0.35
    simulationSpace=RootRelative
    boundBone {boneName=dyn_C_belly}
    componentAppliedLinearAccClamp=<3000 3000 3000>
    componentLinearAccScale=<1 1 1>
    componentLinearVelScale=<1 1 1>
    gravityOverride {z=-980}
    physicsBodyDefinitions [{
      boundBone {boneName=dyn_C_belly}
      boxExtents=<23 23 23>
      constraintSetup {angularLimitsMax=<10 15 20>, angularLimitsMin=<-10 -15 -20>, angularTarget=<1 0 0>}
      localJointOffset=<3 0 0>
     }]
    simSpaceSettings {maxAngularAcceleration=25, maxAngularVelocity=5, simSpaceAngularAlpha=1}
   }
   n5 = Anim Dynamics - Bone: dyn_R_belly(ComponentPose=@n4, Alpha=1, ExternalForce="0,0,0") {
    angularDampingOverride=1
    angularSpringConstant=15000
    bAngularSpring=true
    bOverrideAngularDamping=true
    bUseGravityOverride=true
    gravityScale=0.35
    simulationSpace=RootRelative
    boundBone {boneName=dyn_R_belly}
    componentAppliedLinearAccClamp=<3000 3000 3000>
    componentLinearAccScale=<1 1 1>
    componentLinearVelScale=<1 1 1>
    gravityOverride {z=-980}
    physicsBodyDefinitions [{
      boundBone {boneName=dyn_R_belly}
      boxExtents=<28 28 28>
      constraintSetup {angularLimitsMax=<10 15 20>, angularLimitsMin=<-10 -15 -20>, angularTarget=<1 0 0>}
      localJointOffset=<-3 0 0>
     }]
    simSpaceSettings {maxAngularAcceleration=25, maxAngularVelocity=5, simSpaceAngularAlpha=1}
   }
   n6 = "Transform (Modify) Bone - Bone: dyn_L_belly"!(ComponentPose=@n5, Alpha=1) -> (Pose) {
    translationMode=BMM_Additive
    translationSpace=BCS_BoneSpace
    alphaScaleBiasClamp {bias=1, scale=-1}
    boneToModify {boneName=dyn_L_belly}
    translation {z=3}
   }
   n7 = "Transform (Modify) Bone - Bone: dyn_C_belly"!(ComponentPose=@n6, Alpha=1) -> (Pose) {
    translationMode=BMM_Additive
    translationSpace=BCS_BoneSpace
    alphaScaleBiasClamp {bias=1, scale=-1}
    boneToModify {boneName=dyn_C_belly}
    translation {z=3}
   }
   n8 = "Transform (Modify) Bone - Bone: dyn_R_belly"!(ComponentPose=@n7, Alpha=1) -> (Pose) {
    translationMode=BMM_Additive
    translationSpace=BCS_BoneSpace
    alphaScaleBiasClamp {bias=1, scale=-1}
    boneToModify {boneName=dyn_R_belly}
    translation {z=-3}
   }
   n9 = Anim Dynamics - Bone: dyn_L_belly!(ComponentPose=@n8, Alpha=1, GravityOverride:Vector="0,0,-200") -> (Pose) {
    angularBiasOverride=1
    angularDampingOverride=0.99
    angularSpringConstant=10000
    bGravityOverrideInSimSpace=true
    bLinearSpring=true
    bOverrideAngularBias=true
    bOverrideLinearDamping=true
    bUseGravityOverride=true
    linearDampingOverride=1
    linearSpringConstant=300
    simulationSpace=RootRelative
    alphaScaleBiasClamp {bias=1, scale=-1}
    boundBone {boneName=dyn_L_belly}
    componentAppliedLinearAccClamp=<5 5 5>
    gravityOverride {z=-200}
    physicsBodyDefinitions [{
      boundBone {boneName=dyn_L_belly}
      boxExtents=<150 150 150>
      constraintSetup {angularLimitsMax=<0 0 0>, angularLimitsMin=<0 0 0>, angularTarget=<1 0 0>, linearAxesMax=<0 0.5 0.5>, linearAxesMin=<0 -0.5 -3>}
      localJointOffset=<0.2 0 0>
     }]
    relativeSpaceBone {boneName=spine_04}
    simSpaceSettings {maxAngularAcceleration=50, maxAngularVelocity=10, simSpaceAngularAlpha=0.1}
   }
   n10 = Anim Dynamics - Bone: dyn_C_belly!(ComponentPose=@n9, Alpha=1, GravityOverride:Vector="0,0,-200") -> (Pose) {
    angularBiasOverride=1
    angularDampingOverride=0.99
    angularSpringConstant=10000
    bGravityOverrideInSimSpace=true
    bLinearSpring=true
    bOverrideAngularBias=true
    bOverrideLinearDamping=true
    bUseGravityOverride=true
    linearDampingOverride=1
    linearSpringConstant=180
    simulationSpace=RootRelative
    alphaScaleBiasClamp {bias=1, scale=-1}
    boundBone {boneName=dyn_C_belly}
    componentAppliedLinearAccClamp=<5 5 5>
    gravityOverride {z=-200}
    physicsBodyDefinitions [{
      boundBone {boneName=dyn_C_belly}
      boxExtents=<80 80 80>
      constraintSetup {angularLimitsMax=<0 0 0>, angularLimitsMin=<0 0 0>, angularTarget=<1 0 0>, linearAxesMax=<0 0.5 0.5>, linearAxesMin=<0 -0.5 -3>}
      localJointOffset=<0.2 0 0>
     }]
    relativeSpaceBone {boneName=spine_04}
    simSpaceSettings {maxAngularAcceleration=50, maxAngularVelocity=10, simSpaceAngularAlpha=0.1}
   }
   n11 = Anim Dynamics - Bone: dyn_R_belly!(ComponentPose=@n10, Alpha=1, GravityOverride:Vector="0,0,-200") -> (Pose) {
    angularBiasOverride=1
    angularDampingOverride=0.99
    angularSpringConstant=10000
    bGravityOverrideInSimSpace=true
    bLinearSpring=true
    bOverrideAngularBias=true
    bOverrideLinearDamping=true
    bUseGravityOverride=true
    linearDampingOverride=1
    linearSpringConstant=300
    simulationSpace=RootRelative
    alphaScaleBiasClamp {bias=1, scale=-1}
    boundBone {boneName=dyn_R_belly}
    componentAppliedLinearAccClamp=<5 5 5>
    gravityOverride {z=-200}
    physicsBodyDefinitions [{
      boundBone {boneName=dyn_R_belly}
      boxExtents=<150 150 150>
      constraintSetup {angularLimitsMax=<0 0 0>, angularLimitsMin=<0 0 0>, angularTarget=<1 0 0>, linearAxesMax=<0 0.5 2>, linearAxesMin=<0 -0.5 -0.5>}
      localJointOffset=<-0.2 0 0>
     }]
    relativeSpaceBone {boneName=spine_04}
    simSpaceSettings {maxAngularAcceleration=50, maxAngularVelocity=10, simSpaceAngularAlpha=0.1}
   }
   Output Pose(Result=Component To Local(ComponentPose=@n11)) {layerGroup=None}
  ]}
 EventGraph {nodes [
   Event BlueprintUpdateAnimation()
   n1 = GetOwnerSkeletalMeshComponent()
   n2 = BreakRotator(InRot=BreakTransform(InTransform=GetDeltaTransformFromRefPose(self=@n1, BoneName=neck_02, BaseName=spine_05)).Rotation)
   > Set neck02 rotate(neck02 rotate:Rotator=MakeRotator(
     Roll=float * float!(A=@n2, B=-1, C:double=0.95) -> (ReturnValue)
     Pitch=float * float(A=@n2.Pitch, B=-1, C=0.95)
     Yaw=float * float(A=@n2.Yaw, B=-1, C=0.95)
    )) -> (Output_Get:Rotator)
  ]}
]
variables [neck02 rotate {type=Rotator}]
