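# Physics asset for the IslandDepth medium male player character body, giving its skeleton simplified collision and joint physics.
# One capsule body per major bone covers pelvis, spine, head and limbs, with a tiny capsule at neck_02 and a zero-length, effectively spherical one at the pelvis.
# Constraints chain the bodies out from the pelvis with uniform swing and twist angles, but elbows, neck and head are set to rotate freely.
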PhysicsAsset /BRCosmetics/Characters/Player/Male/Medium/Bodies/M_MED_IslandDepth/Meshes/PA_Character_IslandDepth
bodies {
 calf_l [{
   shapeType=Capsule
   length=7.9962
   radius=17.4061
   shapeName=calf_l_capsule
   center=<-12.7208 4.3631 1.1525>
   rotation=<rot -86.3496 -61.4643 56.6005>
  }]
 calf_r [{
   shapeType=Capsule
   length=7.5261
   radius=17.5699
   shapeName=calf_r_capsule
   center=<12.598 -4.3337 -1.2011>
   rotation=<rot -84.2242 -79.8228 -104.4416>
  }]
 foot_l [{
   shapeType=Capsule
   length=10.4901
   radius=15.0134
   shapeName=foot_l_capsule
   center=<-1.5686 12.0161 0.5076>
   rotation=<rot -4.6065 19.9577 85.2556>
  }]
 foot_r [{
   shapeType=Capsule
   length=10.3853
   radius=14.9955
   shapeName=foot_r_capsule
   center=<1.5056 -12.1213 -0.3981>
   rotation=<rot -5.0128 20.7177 -93.8969>
  }]
 hand_l [{
   shapeType=Capsule
   length=7.6847
   radius=9.1419
   shapeName=hand_l_capsule
   center=<5.2613 -1.7034 0.1877>
   rotation=<rot 87.4306 -41.7791 -36.3467>
  }]
 hand_r [{
   shapeType=Capsule
   length=7.6847
   radius=9.1419
   shapeName=hand_r_capsule
   center=<-5.2613 1.7034 -0.1877>
   rotation=<rot 87.4306 -41.7791 143.6533>
  }]
 head [{
   shapeType=Capsule
   length=6.2214
   radius=23.0272
   shapeName=head_capsule
   center=<-3.3474 7.2067 0.6139>
   rotation=<rot 1.0064 8.3407 -89.2253>
  }]
 lowerarm_l [{
   shapeType=Capsule
   length=6.4895
   radius=11.6265
   shapeName=lowerarm_l_capsule
   center=<10.5493 0.1323 2.2575>
   rotation=<rot 85.0757 -10.4412 175.6605>
  }]
 lowerarm_r [{
   shapeType=Capsule
   length=6.4895
   radius=11.6265
   shapeName=lowerarm_r_capsule
   center=<-10.5493 -0.1323 -2.2575>
   rotation=<rot 85.0757 -10.4412 -4.3395>
  }]
 neck_02 [{shapeType=Capsule, length=1, radius=1}]
 pelvis [{
   shapeType=Capsule
   length=0
   radius=40.5155
   shapeName=pelvis_capsule
   center=<10.9573 9.7709 0.0735>
   rotation=<rot -89.0166 -146.5541 148.4961>
  }]
 spine_03 [{
   shapeType=Capsule
   length=12.3905
   radius=23.7122
   shapeName=spine_03_capsule
   center=<6.3156 3.0857 0.2113>
   rotation=<rot -0.3814 20.6921 -0.1236>
  }]
 spine_05 [{
   shapeType=Capsule
   length=4.2126
   radius=25.96
   shapeName=spine_05_capsule
   center=<-5.3316 6.8699 -1.9498>
   rotation=<rot -4.7455 9.5604 -0.5946>
  }]
 thigh_l [{
   shapeType=Capsule
   length=13
   radius=18.3994
   shapeName=thigh_l_capsule
   center=<-16.4806 -0.7693 -0.3355>
   rotation=<rot -81.5295 -107.7647 109.3361>
  }]
 thigh_r [{
   shapeType=Capsule
   length=10
   radius=21.3994
   shapeName=thigh_r_capsule
   center=<16.4806 0.7693 0.3355>
   rotation=<rot -81.5295 -107.7647 -70.6639>
  }]
 upperarm_l [{
   shapeType=Capsule
   length=13.7835
   radius=11.5165
   shapeName=upperarm_l_capsule
   center=<13.5368 -0.1492 -0.81>
   rotation=<rot -7.4425 94.1297 -78.2741>
  }]
 upperarm_r [{
   shapeType=Capsule
   length=12.3375
   radius=11.6341
   shapeName=upperarm_r_capsule
   center=<-13.3188 0.1553 0.8572>
   rotation=<rot -8.6824 92.5631 102.8761>
  }]
}
constraints [
 {bone1Name=spine_03, bone2Name=pelvis, swing1LimitDegrees=45, swing2LimitDegrees=45, twistLimitDegrees=45}
 {bone1Name=spine_05, bone2Name=spine_03, swing1LimitDegrees=45, swing2LimitDegrees=45, twistLimitDegrees=45}
 {bone1Name=upperarm_l, bone2Name=spine_05, swing1LimitDegrees=45, swing2LimitDegrees=45, twistLimitDegrees=45}
 {bone1Name=hand_l, bone2Name=lowerarm_l, swing1LimitDegrees=45, swing2LimitDegrees=45, twistLimitDegrees=45}
 {bone1Name=upperarm_r, bone2Name=spine_05, swing1LimitDegrees=45, swing2LimitDegrees=45, twistLimitDegrees=45}
 {bone1Name=hand_r, bone2Name=lowerarm_r, swing1LimitDegrees=45, swing2LimitDegrees=45, twistLimitDegrees=45}
 {bone1Name=thigh_l, bone2Name=pelvis, swing1LimitDegrees=45, swing2LimitDegrees=45, twistLimitDegrees=45}
 {bone1Name=calf_l, bone2Name=thigh_l, swing1LimitDegrees=45, swing2LimitDegrees=45, twistLimitDegrees=45}
 {bone1Name=foot_l, bone2Name=calf_l, swing1LimitDegrees=45, swing2LimitDegrees=45, twistLimitDegrees=45}
 {bone1Name=thigh_r, bone2Name=pelvis, swing1LimitDegrees=45, swing2LimitDegrees=45, twistLimitDegrees=45}
 {bone1Name=calf_r, bone2Name=thigh_r, swing1LimitDegrees=45, swing2LimitDegrees=45, twistLimitDegrees=45}
 {bone1Name=foot_r, bone2Name=calf_r, swing1LimitDegrees=45, swing2LimitDegrees=45, twistLimitDegrees=45}
 {
  bone1Name=lowerarm_r
  bone2Name=upperarm_r
  swing1LimitDegrees=45
  swing1Motion=Free
  swing2LimitDegrees=45
  swing2Motion=Free
  twistLimitDegrees=45
  twistMotion=Free
 }
 {
  bone1Name=lowerarm_l
  bone2Name=upperarm_l
  swing1LimitDegrees=45
  swing1Motion=Free
  swing2LimitDegrees=45
  swing2Motion=Free
  twistLimitDegrees=45
  twistMotion=Free
 }
 {
  bone1Name=neck_02
  bone2Name=spine_05
  swing1LimitDegrees=45
  swing1Motion=Free
  swing2LimitDegrees=45
  swing2Motion=Free
  twistLimitDegrees=45
  twistMotion=Free
 }
 {
  bone1Name=head
  bone2Name=neck_02
  swing1LimitDegrees=45
  swing1Motion=Free
  swing2LimitDegrees=45
  swing2Motion=Free
  twistLimitDegrees=45
  twistMotion=Free
 }
]
