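# Physics asset for the IslandDepth medium male player body cosmetic, defining per-bone collision volumes for the character.
# Each body is a single fitted capsule, covering head, pelvis, two spine bones, and left and right arms, hands, legs and feet.
# The neck_02 capsule is a tiny unit-sized shape, and several pelvis and leg capsules carry negative lengths.
# No constraints are defined, so no joints link the bodies.
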PhysicsAsset /BRCosmetics/Characters/Player/Male/Medium/Bodies/M_MED_IslandDepth/Meshes/PA_Relative_Character_IslandDepth
bodies {
 calf_l [{
   shapeType=Capsule
   length=-0.6303
   radius=36.0831
   shapeName=calf_l_capsule
   center=<17.3398 -1.6897 10.9478>
   rotation=<rot -71.3757 -25.9152 46.7831>
  }]
 calf_r [{
   shapeType=Capsule
   length=1.4952
   radius=35.2471
   shapeName=calf_r_capsule
   center=<-15.2075 1.5782 -10.8108>
   rotation=<rot -70.7526 -35.7321 -125.5526>
  }]
 foot_l [{
   shapeType=Capsule
   length=17.4366
   radius=18.6214
   shapeName=foot_l_capsule
   center=<-0.3143 9.5251 0.0986>
   rotation=<rot -6.5963 17.6376 85.687>
  }]
 foot_r [{
   shapeType=Capsule
   length=17.4787
   radius=18.6043
   shapeName=foot_r_capsule
   center=<0.3715 -9.4905 -0.02>
   rotation=<rot -6.9144 17.5361 -94.1098>
  }]
 hand_l [{
   shapeType=Capsule
   length=8.6187
   radius=6.1692
   shapeName=hand_l_capsule
   center=<5.2767 -0.5189 -0.5906>
   rotation=<rot 86.4508 149.4709 155.9179>
  }]
 hand_r [{
   shapeType=Capsule
   length=8.7137
   radius=6.1578
   shapeName=hand_r_capsule
   center=<-5.329 0.6292 0.6185>
   rotation=<rot 86.7912 157.2926 -15.9593>
  }]
 head [{
   shapeType=Capsule
   length=0.6986
   radius=24.0074
   shapeName=head_capsule
   center=<-3.1404 7.2928 -0.0115>
   rotation=<rot 1.4453 7.5846 -89.3961>
  }]
 lowerarm_l [{
   shapeType=Capsule
   length=4.1814
   radius=13.1051
   shapeName=lowerarm_l_capsule
   center=<3.8513 1.1715 0.9039>
   rotation=<rot 81.099 -100.3033 86.1137>
  }]
 lowerarm_r [{
   shapeType=Capsule
   length=2.8193
   radius=13.2178
   shapeName=lowerarm_r_capsule
   center=<-3.6471 -1.2236 -0.9206>
   rotation=<rot 80.9489 -105.9435 -99.6187>
  }]
 neck_02 [{shapeType=Capsule, length=1, radius=1}]
 pelvis [{
   shapeType=Capsule
   length=-17.696
   radius=41.4819
   shapeName=pelvis_capsule
   center=<10.8917 4.3297 -0.2103>
   rotation=<rot -87.6659 66.107 -72.6684>
  }]
 spine_03 [{
   shapeType=Capsule
   length=16.1077
   radius=23.534
   shapeName=spine_03_capsule
   center=<12.2843 5.6262 0.0658>
   rotation=<rot -0.0168 13.9225 0.0326>
  }]
 spine_05 [{
   shapeType=Capsule
   length=7.922
   radius=28.2639
   shapeName=spine_05_capsule
   center=<-12.1898 7.7834 0.007>
   rotation=<rot 0.0217 -14.2396 0.0111>
  }]
 thigh_l [{
   shapeType=Capsule
   length=-18.6304
   radius=20.3716
   shapeName=thigh_l_capsule
   center=<-14.22 -1.2366 1.5712>
   rotation=<rot -66.6687 -20.6664 19.4138>
  }]
 thigh_r [{
   shapeType=Capsule
   length=-32.833
   radius=24.6334
   shapeName=thigh_r_capsule
   center=<9.1338 -0.6712 0.2193>
   rotation=<rot -59.4506 -27.1339 -161.985>
  }]
 upperarm_l [{
   shapeType=Capsule
   length=10.9772
   radius=13.2267
   shapeName=upperarm_l_capsule
   center=<11.2396 -0.1936 -1.9507>
   rotation=<rot 4.8845 92.9092 -70.5933>
  }]
 upperarm_r [{
   shapeType=Capsule
   length=10.9663
   radius=13.2242
   shapeName=upperarm_r_capsule
   center=<-11.2555 0.221 1.9432>
   rotation=<rot 5.1692 92.977 109.294>
  }]
}
constraints []
